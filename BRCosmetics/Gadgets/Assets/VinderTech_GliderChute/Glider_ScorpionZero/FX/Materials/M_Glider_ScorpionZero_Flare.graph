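# Additive unlit Niagara mesh-particle material for the flare effect on the Scorpion Zero glider.
# Vertex math stretches the mesh along its local Y axis by UV, thickens it with SplineThicken and shifts it along the camera direction.
# Two counter-rotating spikeball flare textures are screen-blended, tinted by particle color and passed through MF_FX_EmissiveAdjustments.
# Alpha gets depth fade, skipped on Low and Medium quality.
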Material /BRCosmetics/Gadgets/Assets/VinderTech_GliderChute/Glider_ScorpionZero/FX/Materials/M_Glider_ScorpionZero_Flare
graph {
 properties {
  bUsedWithNiagaraMeshParticles=true
  blendMode=BLEND_Additive
  pixelDepthOffsetMode=PDOM_Legacy
  shadingModel=MSM_Unlit
  usedShadingModels=Unlit
 }
 nodes [
  n1 = Subtract(A=TextureCoordinate()) {constB=0.5}
  n2 = Multiply(
   A=Multiply(A=ComponentMask(None=@n1) {g=true}, B=ScalarParameter() {defaultValue=222, parameterName=Additional Length})
   B=Transform(None=Constant3Vector() {constant=<rgba 0 1 0 0>})
  )
  n3 = ScalarParameter() {defaultValue=166, parameterName=Width}
  n4 = MaterialFunctionCall(WidthBase=@n3, WidthTip=@n3, WorldPosition=Add(A=@n2, B=WorldPosition().XYZ)) {materialFunction=&SplineThicken}
  n5 = WorldPosition()
  n6 = Normalize(VectorInput=Subtract(A=CameraPositionWS(), B=Add(A=@n4.WorldPositionOffset, B=@n5.XYZ)))
  n7 = ScalarParameter() {defaultValue=-66, parameterName=Camera Offset}
  Add(A=@n2, B=Add(A=@n4.WorldPositionOffset, B=Multiply(A=@n6, B=@n7)))
  CameraVectorWS()
  n8 = ParticleColor()
  n9 = DepthFade(Opacity=@n8.A, FadeDistance=ScalarParameter() {defaultValue=33, parameterName=Depth Fade})
  n10 = TextureSample(UVs=Rotator() {speed=0.618}) {samplerSource=SSM_Clamp_WorldGroupSettings, texture=/Game/Effects/Fort_Effects/Textures/Flares/T_Flare_Spikeball}
  n11 = TextureSample(UVs=Rotator() {speed=-0.381}) {samplerSource=SSM_Clamp_WorldGroupSettings, texture=&T_Flare_Spikeball}
  n12 = OneMinus(None=Multiply(A=OneMinus(None=@n10.RGB), B=OneMinus(None=@n11.RGB)))
  MaterialFunctionCall(Emissive=Multiply(A=QualitySwitch(Default=@n9, Low=@n8.A, Medium=@n8.A), B=Multiply(A=@n8.RGB, B=@n12))) {materialFunction=&MF_FX_EmissiveAdjustments}
  n13 = DotProduct(A=@n4.Normal, B=@n6)
  VertexInterpolator(VS=Power(Base=Abs(None=@n13), Exp=ScalarParameter() {defaultValue=2, parameterName=Fresnel Exponent}))
 ]
}
