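# Additive unlit material for Niagara mesh particles on the Scorpion Zero glider, drawing a camera-facing rainbow flare.
# A Sprite function billboards the mesh, scaled by dynamic SpriteScale, with an added offset along the camera vector.
# Radial UV distance and dynamic RainbowThickness form a soft ring blending three colors.
# Two inverted, counter-rotating spikeball textures multiply and are raised to a power, breaking the ring into shimmering star rays.
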Material /BRCosmetics/Gadgets/Assets/VinderTech_GliderChute/Glider_ScorpionZero/FX/Materials/M_Glider_ScorpionZero_Flare_Rainbow
graph {
 properties {
  bUsedWithNiagaraMeshParticles=true
  blendMode=BLEND_Additive
  pixelDepthOffsetMode=PDOM_Legacy
  shadingModel=MSM_Unlit
  usedShadingModels=Unlit
 }
 nodes [
  n1 = WorldPosition()
  n2 = MaterialFunctionCall!(
   Center Location=@n1.XYZ
   0-1 UVs
   XY scale=DynamicParameter() {parameterIndex=1, paramNames [SpriteScale, RainbowThickness, Opacity, ""]}.SpriteScale
   OverridePivot
   "Normals (Optional)"
  ) {materialFunction=&Sprite}
  n3 = Subtract(A=CameraPositionWS(), B=Add(A=@n1.XYZ, B=@n2.World position offset))
  Add(A=Multiply(A=Normalize(VectorInput=@n3), B=ScalarParameter() {defaultValue=-66, parameterName=Camera Offset}), B=@n2.World position offset)
  DynamicParameter() {parameterIndex=1, paramNames [SpriteScale, RainbowThickness, Opacity, ""]}
  n4 = VectorParameter() {parameterName=Color and Roughness 1, defaultValue=<rgba 0.8796 0.2307 0.0723 1>}
  n5 = VectorParameter() {parameterName=Color and Roughness 2, defaultValue=<rgba 0.7835 0.8388 0.2016 1>}
  n6 = VectorParameter() {parameterName=Color and Roughness 3, defaultValue=<rgba 0.2423 0.8714 0.4793 1>}
  n7 = Constant() {r=0.5}
  n8 = Distance(A=TextureCoordinate(), B=@n7)
  n9 = Saturate(None=Divide(A=Divide(A=@n8, B=@n7), B=ScalarParameter() {defaultValue=0.8, parameterName=Rainbow Radius}))
  n10 = Saturate(None=Divide(A=OneMinus(None=@n9), B=DynamicParameter() {parameterIndex=1, paramNames [SpriteScale, RainbowThickness, Opacity, ""]}.RainbowThickness))
  n11 = ConstantBiasScale(None=@n10) {bias=-0.5, scale=2}
  n12 = Power(Base=OneMinus(None=DotProduct(A=@n11, B=@n11))) {desc=Rainbow Opacity}
  n13 = MaterialFunctionCall(UVs=TextureCoordinate(), Texture Scale=ScalarParameter() {defaultValue=3, parameterName=Flare Star Scale}) {materialFunction=&ScaleUVsByCenter}
  n14 = OneMinus(None=TextureSample(UVs=Rotator(Coordinate=@n13.UVs) {speed=0.618}) {samplerSource=SSM_Clamp_WorldGroupSettings, texture=/Game/Effects/Fort_Effects/Textures/Flares/T_Flare_Spikeball}.R)
  n15 = OneMinus(None=TextureSample(UVs=Rotator(Coordinate=@n13.UVs) {speed=-0.381}) {samplerSource=SSM_Clamp_WorldGroupSettings, texture=&T_Flare_Spikeball}.R)
  Multiply(
   A=Multiply(
    A=MaterialFunctionCall(1=@n4.RGB, 2=@n5.RGB, 3=@n6.RGB, A=@n10) {materialFunction=&Lerp_Multiple_Float3}.Lerp 3 Inputs
    B=Saturate(None=@n12)
   )
   B=Power(Base=Multiply(A=@n14, B=@n15), Exp=ScalarParameter() {defaultValue=6, parameterName=Flare Star Exponent}) {constExponent=5}
  )
  ScalarParameter() {defaultValue=0.4, parameterName=Rainbow Thickness}
 ]
}
