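# Unlit alpha-composite material for the ScorpionZero glider's trail effect, flagged for Niagara mesh particles.
# A soft layer blends three panned smoke samples, each tinted and time-offset, under a reflected gradient.
# A premultiplied shard layer erodes a panning trail texture with sharp and soft edges for colored shards and glow, driven by edge fades and particle Trail Opacity.
# SplineThicken handles trail width; three unrelated textures feed a VertexInterpolator marked "to prevent cleanup".
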Material /BRCosmetics/Gadgets/Assets/VinderTech_GliderChute/Glider_ScorpionZero/FX/Materials/M_Glider_ScorpionZero_Trail
graph {
 properties {
  bUsedWithNiagaraMeshParticles=true
  blendMode=BLEND_AlphaComposite
  pixelDepthOffsetMode=PDOM_Legacy
  shadingModel=MSM_Unlit
  usedShadingModels=Unlit
 }
 nodes [
  n1 = TextureCoordinate()
  n2 = MaterialFunctionCall!(Gradient=ComponentMask(None=@n1) {r=true}, Center Point=ScalarParameter() {defaultValue=0.3, parameterName=Soft Trail Gradient Center}) {materialFunction=/Game/Effects/Fort_Effects/MaterialFunctions/MF_GradientReflected}
  n3 = ScalarParameter() {defaultValue=2, parameterName=Soft Trail Opacity Exponent}
  n4 = VectorParameter() {parameterName=Color 1, defaultValue=<rgba 0.8796 0.2307 0.0723 1>}
  n5 = VectorParameter() {parameterName=Soft Trail U Scale Speed Time Offsets, defaultValue=<rgba 0.2 -0.333 0.333 0.6>}
  n6 = Multiply(A=@n1, B=AppendVector(A=@n5.R, B=Constant() {r=1}))
  n7 = AppendVector(A=@n5.G, B=Constant())
  n8 = TextureSample(UVs=Panner(Coordinate=@n6, Speed=@n7) {bFractionalPart=true}) {texture=/Game/Effects/Fort_Effects/Textures/Smoke/T_Smoke_Trail_Soft_2}
  n9 = Multiply(A=@n4.RGB, B=Multiply(A=@n8.R, B=@n8.R))
  n10 = Add(A=Time(), B=ParticleRandom())
  n11 = TextureSample(UVs=Panner(Coordinate=@n6, Time=Add(A=@n10, B=@n5.B), Speed=@n7) {bFractionalPart=true}) {texture=&T_Smoke_Trail_Soft_2}
  n12 = VectorParameter() {parameterName=Color 2, defaultValue=<rgba 0.7835 0.8388 0.2016 1>}
  n13 = Multiply(A=Multiply(A=@n11.R, B=@n11.R), B=@n12.RGB)
  n14 = TextureSample(UVs=Panner(Coordinate=@n6, Time=Add(A=@n10, B=@n5.A), Speed=@n7) {bFractionalPart=true}) {texture=&T_Smoke_Trail_Soft_2}
  n15 = Multiply(A=@n14.R, B=@n14.R)
  n16 = VectorParameter() {parameterName=Color 3, defaultValue=<rgba 0.2423 0.8714 0.4793 1>}
  n17 = Divide(A=Add(A=Add(A=@n9, B=@n13), B=Multiply(A=@n15, B=@n16.RGB))) {constB=3}
  n18 = Multiply(
   A=ScalarParameter() {defaultValue=0.666, parameterName=Soft Trail Intensity}
   B=DynamicParameter() {parameterIndex=1, paramNames [Trail Opacity, "", "", ""]}.Trail Opacity
  )
  n19 = VectorParameter() {parameterName=Glow Color, defaultValue=<rgba 1 0.8177 0 1>}
  n20 = TextureCoordinate()
  n21 = ComponentMask(None=@n20) {r=true}
  n22 = MaterialFunctionCall(1=@n4.RGB, 2=@n12.RGB, 3=@n16.RGB, A=Multiply(A=@n21, B=ScalarParameter() {defaultValue=2.2, parameterName=Color Lerp Scale})) {materialFunction=&Lerp_Multiple_Float3}
  n23 = Add(A=TextureCoordinate(), B=ParticleRandom())
  n24 = VectorParameter() {parameterName=UV Scale and speed, defaultValue=<rgba 1.1 0.65 -1.63 0.2>}
  n25 = ComponentMask(None=@n24.RGB) {g=true, r=true}
  n26 = TextureSample(UVs=Panner(Coordinate=Multiply(A=@n23, B=@n25), Speed=AppendVector(A=@n24.B, B=@n24.B)) {bFractionalPart=true}) {
   samplerType=SAMPLERTYPE_LinearGrayscale
   texture=/BRCosmetics/Gadgets/Assets/VinderTech_GliderChute/Glider_ScorpionZero/FX/Textures/T_Glider_ScorpionZero_Trail
  }
  n27 = ConstantBiasScale(None=@n20) {bias=-0.5, scale=2}
  n28 = Divide(
   A=OneMinus(None=Abs(None=@n27))
   B=AppendVector(A=ScalarParameter() {defaultValue=1, parameterName=Length Fade Sze}, B=ScalarParameter() {defaultValue=3, parameterName=Width Fade Size})
  )
  n29 = MaterialFunctionCall(Float2=OneMinus(None=Saturate(None=@n28))) {materialFunction=&BreakOutFloat2Components}
  n30 = ScalarParameter() {defaultValue=0.9, parameterName=Erosion Max}
  n31 = ScalarParameter() {defaultValue=0.8, parameterName=Erosion Min}
  n32 = MaterialFunctionCall(Gradient=@n21, Center Point=ScalarParameter() {defaultValue=0.4, parameterName=Erosion Gradient Center Point}) {materialFunction=&MF_GradientReflected}
  n33 = Power(Base=DynamicParameter() {parameterIndex=1, paramNames [Trail Opacity, "", "", ""]}.Trail Opacity) {constExponent=4}
  n34 = Max(A=Max(A=Max(A=@n29.R, B=@n29.R), B=LinearInterpolate(A=@n30, B=@n31, Alpha=@n32.Result)), B=OneMinus(None=Saturate(None=@n33)))
  n35 = MaterialFunctionCall!(Gradient=@n26.R, Erosion Amount=@n34, Softness=ScalarParameter() {defaultValue=0.1, parameterName=Erosion Softness Shards}) {materialFunction=/BRCosmetics/Athena/Effects/Particles/Skydiving/NS_Contrail_DarkAndLight/Functions/MF_Erosion}
  n36 = MaterialFunctionCall(Gradient=@n26.R, Erosion Amount=@n34, Softness=ScalarParameter() {defaultValue=1, parameterName=Erosion Softness Glow}) {materialFunction=&MF_Erosion}
  n37 = Multiply(A=OneMinus(None=@n29.R), B=Max(A=@n35.Increased Range, B=Power(Base=@n36.Increased Range)))
  n38 = ScalarParameter() {defaultValue=0.2, parameterName=Opacity Shards}
  Add(
   A=Multiply(A=Power(Base=@n2.Result, Exp=@n3), B=Multiply(A=@n17, B=@n18))
   B=Multiply(A=LinearInterpolate(A=@n19.RGB, B=@n22.Lerp 3 Inputs, Alpha=@n35.Increased Range), B=Multiply(A=@n37, B=@n38)) {desc=Premultiplying shards}
  )
  MaterialFunctionCall(
   WidthBase=ScalarParameter() {defaultValue=44, parameterName=Width Base}
   WidthTip=ScalarParameter() {defaultValue=11, parameterName=Width Tip}
   Expand U or V UV Channel=StaticBoolParameter() {parameterName="Expand U or V UV Channel (SB)"}
  ) {materialFunction=&SplineThicken}
  MaterialFunctionCall() {materialFunction=&MF_FX_EmissiveAdjustments}
  n39 = TextureSample() {
   samplerType=SAMPLERTYPE_LinearGrayscale
   texture=/Game/Gadgets/Assets/VinderTech_GliderChute/Glider_Dazzle/FX/Variant/T_Glider_Dazzle_Electric_Crystal_Noise
  }
  n40 = TextureSample() {texture=/BRCosmetics/Athena/Effects/Particles/Skydiving/NS_Contrail_PinkSpike/Textures/T_PinkSpike_Voronoi2D_Height}
  n41 = TextureSample() {
   samplerType=SAMPLERTYPE_LinearGrayscale
   texture=/BRCosmetics/Accessories/FORT_Backpacks/Backpack_M_MED_FearlessFlightHero/FX/Textures/T_FFH_FragmentBevel_M
  }
  VertexInterpolator(VS=Add(A=Add(A=@n39.R, B=@n40.R), B=@n41.R)) {desc=to prevent cleanup}
 ]
}
