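# Material function that animates wings on the Scorpion Zero glider's mesh-particle FX, outputting World Position Offset, corrected normals and local-bounds gradients.
# A pre-skinned-bounds Y gradient masks a travelling sine flap, which fades as rotation intensity rises, plus an oscillating rotation around the particle position.
# Curl Factor blends that rotation toward a curl around an offset pivot.
# Dynamic Parameters drive direction, time offset, intensity and curl, while FixRotateAboutAxisNormals rebuilds normals for both rotations.
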MaterialFunction /BRCosmetics/Gadgets/Assets/VinderTech_GliderChute/Glider_ScorpionZero/FX/Materials/MF_Glider_ScorpionZero_WPO
graph {nodes [
  n1 = WorldPosition() {worldPositionShaderOffset=WPT_ExcludeAllShaderOffsets}
  n2 = PreSkinnedLocalBounds()
  n3 = Divide(
   A=Subtract(A=TransformPosition(None=@n1.XYZ) {transformSourceType=TRANSFORMPOSSOURCE_World, transformType=TRANSFORMPOSSOURCE_Instance}, B=@n2.Min)
   B=@n2.Extents
  )
  n4 = ComponentMask(None=@n3) {g=true}
  n5 = Multiply(A=@n4, B=ScalarParameter() {defaultValue=0.62, parameterName=Wavelength})
  n6 = Time()
  n7 = DynamicParameter() {defaultValue=<rgba 1 1 1 0>, paramNames [Wing Rotation Direction, Wing Rotation Time Offset, Wing Rotation Intensity, Wing Curl Factor]}
  n8 = Add(A=@n6, B=@n7.Wing Rotation Time Offset)
  n9 = Multiply(A=@n8, B=ScalarParameter() {defaultValue=1.62, parameterName=Flapping Speed})
  n10 = Multiply(A=Sine(None=Subtract(A=@n5, B=@n9)), B=ScalarParameter() {defaultValue=0.6, parameterName=Distance})
  n11 = SquareRoot(None=Saturate(None=@n4))
  n12 = Normalize(VectorInput=Transform(None=Constant3Vector() {constant=<rgba 0 0 1 0>}) {transformSourceType=TRANSFORMSOURCE_Instance})
  n13 = Subtract(A=@n8, B=Multiply(A=@n4, B=ScalarParameter() {defaultValue=0.2, parameterName=Wing Rotation Offset}))
  n14 = ScalarParameter() {defaultValue=2, parameterName=Wing Rotation Speed}
  n15 = Multiply(A=Sine(None=Multiply(A=@n13, B=@n14)), B=ScalarParameter() {defaultValue=0.033, parameterName=Wing Rotation Range})
  n16 = Sign(None=@n7.Wing Rotation Direction)
  n17 = Multiply(A=Multiply(A=@n7.Wing Rotation Intensity, B=Multiply(A=@n15, B=@n16)), B=@n4)
  n18 = Transform(None=Constant3Vector() {constant=<rgba 1 0 0 0>}) {transformSourceType=TRANSFORMSOURCE_Instance}
  n19 = RotateAboutAxis(NormalizedRotationAxis=@n12, RotationAngle=@n17, PivotPoint=Constant(), Position=@n18)
  n20 = Multiply(A=Multiply(A=Multiply(A=@n10, B=@n11), B=OneMinus(None=@n7.Wing Rotation Intensity)), B=Normalize(VectorInput=Add(A=@n19, B=@n18))) {constB=0}
  n21 = WorldPosition()
  Add(A=@n20, B=@n21.XYZ)
  n22 = ParticlePositionWS()
  n23 = RotateAboutAxis(NormalizedRotationAxis=@n12, RotationAngle=@n17, PivotPoint=@n22, Position=@n21.XYZ)
  n24 = Multiply(A=@n4, B=Multiply(A=Sign(None=@n7.Wing Curl Factor), B=ScalarParameter() {defaultValue=0.166, parameterName=Curl Intensity}))
  n25 = Multiply(A=Constant3Vector() {constant=<rgba 0 1 0 0>}, B=ScalarParameter() {defaultValue=40, parameterName=Curl Pivot Offset})
  n26 = RotateAboutAxis(
   NormalizedRotationAxis=@n12
   RotationAngle=@n24
   PivotPoint=Add(A=@n22, B=Transform(None=@n25) {transformSourceType=TRANSFORMSOURCE_Instance})
   Position=@n21.XYZ
  )
  n27 = NamedRerouteUsage()
  FunctionOutput(None=Add(A=@n20, B=LinearInterpolate(A=@n23, B=@n26, Alpha=@n27))) {outputName=World Position Offset}
  n28 = MaterialFunctionCall!(Rotation Axis=@n12, Rotation Angle=@n17, World Space Vertex Normals, Movement Mask, PivotPoint) {materialFunction=&FixRotateAboutAxisNormals}
  n29 = MaterialFunctionCall(Rotation Axis=@n12, Rotation Angle=@n24) {materialFunction=&FixRotateAboutAxisNormals}
  FunctionOutput(None=Normalize(VectorInput=LinearInterpolate(A=@n28.New World Space Vertex Normal, B=@n29.New World Space Vertex Normal, Alpha=@n27))) {outputName=World Space Normals, sortPriority=1}
  FunctionOutput(None=@n3) {outputName=Local Bounds Gradients, sortPriority=2}
  NamedRerouteDeclaration(None=Saturate(None=Abs(None=@n7.Wing Curl Factor))) {name=Curl Lerp, nodeColor=<rgba 0 0.1797 1 1>}
  PreviousFrameSwitch(Current Frame=@n6, Previous Frame=Subtract(A=@n6, B=DeltaTime()))
 ]}
