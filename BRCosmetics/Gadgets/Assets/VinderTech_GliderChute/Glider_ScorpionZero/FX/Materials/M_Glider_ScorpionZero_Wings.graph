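# Masked, two-sided lit material for the Scorpion Zero glider wing effects, usable on Niagara mesh particles, with an erosion-driven deploy reveal.
# A shared function supplies WPO, normals and bounds gradients, which drive a rotatable LUT or three-color gradient overlaid with pre-skinned normal shading.
# Dynamic parameters scale WPO, emissive and Gold Erosion, which erodes the opacity mask along a second gradient.
# Floored Gold Erosion switches deploy surface values to idle ones.
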Material /BRCosmetics/Gadgets/Assets/VinderTech_GliderChute/Glider_ScorpionZero/FX/Materials/M_Glider_ScorpionZero_Wings
graph {
 properties {
  bTangentSpaceNormal=false
  bUsedWithNiagaraMeshParticles=true
  blendMode=BLEND_Masked
  pixelDepthOffsetMode=PDOM_Legacy
  twoSided=true
  usedShadingModels=Default Lit
 }
 nodes [
  n1 = MaterialFunctionCall() {materialFunction=&/MF_Glider_ScorpionZero_WPO}
  n2 = VertexInterpolator(VS=@n1.Local Bounds Gradients)
  ComponentMask(None=@n2.PS) {b=true}
  n3 = VectorParameter() {parameterName=Color and Roughness Deploy, defaultValue=<rgba 0.7 0.385 0 0.7>}
  n4 = ScalarParameter() {defaultValue=0.166, parameterName=Color Gradient Rotation, sliderMax=1}
  n5 = DotProduct(
   A=ConstantBiasScale(None=ComponentMask(None=@n2.PS) {b=true, g=true}) {bias=-0.5, scale=2}
   B=AppendVector(A=Sine(None=@n4), B=Cosine(None=@n4))
  )
  n6 = ScalarParameter() {defaultValue=0.8, parameterName=Color Gradient Scale}
  n7 = Saturate(None=ConstantBiasScale(None=Multiply(A=@n5, B=@n6)))
  n8 = TextureSample(UVs=AppendVector(A=@n7, B=Constant() {r=0.5})) {
   samplerSource=SSM_Clamp_WorldGroupSettings
   samplerType=SAMPLERTYPE_LinearColor
   texture=/BRCosmetics/Gadgets/Assets/VinderTech_GliderChute/Glider_ScorpionZero/FX/Textures/T_Glider_ScorpionZero_ColorGradient
  }
  n9 = VectorParameter() {parameterName=Color and Roughness 1, defaultValue=<rgba 0.8796 0.2307 0.0723 1>}
  n10 = VectorParameter() {parameterName=Color and Roughness 2, defaultValue=<rgba 0.7835 0.8388 0.2016 1>}
  n11 = VectorParameter() {parameterName=Color and Roughness 3, defaultValue=<rgba 0.2423 0.8714 0.4793 1>}
  n12 = StaticSwitchParameter(
   True=AppendVector(A=@n8.RGB, B=Constant() {r=1})
   False=MaterialFunctionCall!(1=AppendVector(A=@n9.RGB, B=@n9.RGB), 2=AppendVector(A=@n10.RGB, B=@n10.RGB), 3=AppendVector(A=@n11.RGB, B=@n11.RGB), 4, A=@n7) {materialFunction=/Engine/Functions/Engine_MaterialFunctions02/Lerp_Multiple_Float4}.Lerp 3 Inputs
  ) {defaultValue=true, parameterName=Use Color LUT}
  n13 = ComponentMask(None=PreSkinnedNormal()) {b=true}
  n14 = ScalarParameter() {defaultValue=1, parameterName=Emissive Overlay Intensity}
  n15 = VertexInterpolator(VS=ConstantBiasScale(None=Multiply(A=@n13, B=@n14)))
  n16 = AppendVector(
   A=MaterialFunctionCall(Base=@n12, Blend=ComponentMask(None=@n15.PS) {r=true}) {materialFunction=&Blend_Overlay}.Result
   B=ComponentMask(None=@n12) {a=true}
  )
  n17 = DynamicParameter() {parameterIndex=1, paramNames [Gold Erosion, Emissive Intensity, WPO Intensity, Emissive Intensity]}
  n18 = Floor(None=@n17.Gold Erosion)
  n19 = LinearInterpolate(A=AppendVector(A=@n3.RGB, B=@n3.RGB), B=@n16, Alpha=@n18)
  ComponentMask(None=@n19) {b=true, g=true, r=true}
  ComponentMask(None=@n19) {a=true}
  Constant4Vector() {constant=<rgba 0.8796 0.2307 0.0723 1>}
  Constant4Vector() {constant=<rgba 0.7835 0.8388 0.2016 1>}
  Constant4Vector() {constant=<rgba 0.2423 0.8714 0.4793 1>}
  Constant() {r=1}
  n20 = ConstantBiasScale(None=ComponentMask(None=@n2.PS) {b=true, g=true}) {bias=-0.5, scale=2}
  n21 = ScalarParameter() {defaultValue=0.233, parameterName=Fade In Gradient Rotation, sliderMax=1}
  n22 = AppendVector(A=Sine(None=@n21), B=Cosine(None=@n21))
  n23 = ConstantBiasScale(None=Multiply(A=DotProduct(A=@n20, B=@n22), B=ScalarParameter() {defaultValue=1, parameterName=Fade In Gradient Scale}))
  n24 = MaterialFunctionCall!(
   Gradient=OneMinus(None=Saturate(None=@n23))
   Erosion Amount=OneMinus(None=Saturate(None=@n17.Gold Erosion))
   Softness=ScalarParameter() {defaultValue=0.05, parameterName=Softness}
  ) {materialFunction=/BRCosmetics/Athena/Effects/Particles/Skydiving/NS_Contrail_DarkAndLight/Functions/MF_Erosion}
  n25 = Multiply(A=@n24.Increased Range, B=Saturate(None=TwoSidedSign()))
  LinearInterpolate(
   A=Transform(None=Constant3Vector() {constant=<rgba 0 0 1 0>}) {transformSourceType=TRANSFORMSOURCE_Instance}
   B=VertexInterpolator(VS=@n1.World Space Normals).PS
   Alpha=@n25
  )
  n26 = VectorParameter() {parameterName="Metallic and Specular Deploy (RG) Idle (BA)", defaultValue=<rgba 1 0 0 1>}
  LinearInterpolate(A=@n26.R, B=@n26.R, Alpha=@n18) {constA=1, constB=0}
  LinearInterpolate(A=@n26.G, B=@n26.G, Alpha=@n18)
  n27 = Multiply(A=ComponentMask(None=@n16) {b=true, g=true, r=true}, B=@n17.Emissive Intensity)
  n28 = VectorParameter() {parameterName=Reveal Edge Emissive Color, defaultValue=<rgba 3 2.25 1.5 0>}
  n29 = Multiply(A=@n18, B=@n27)
  MaterialFunctionCall(Emissive=StaticSwitchParameter(True=@n27, False=LinearInterpolate(A=@n28.RGB, B=@n29, Alpha=@n25)) {parameterName=Use Simplified Material}) {materialFunction=&MF_FX_EmissiveAdjustments}
  NamedRerouteDeclaration(None=Ceil(None=@n24.Increased Range)) {name=Opacity Mask, nodeColor=<rgba 0.0781 1 0 1>}
  NamedRerouteUsage()
  n30 = Multiply(A=@n17.WPO Intensity, B=@n1.World Position Offset)
  PreviousFrameSwitch(Current Frame=@n30, Previous Frame=@n30)
  Saturate(None=@n5)
 ]
}
