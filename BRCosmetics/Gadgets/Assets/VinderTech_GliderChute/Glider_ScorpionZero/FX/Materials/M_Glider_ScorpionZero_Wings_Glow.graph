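# Additive unlit Niagara mesh-particle material for the wing glow on the Scorpion Zero glider.
# The glider's shared WPO function, scaled by a dynamic parameter, offsets vertices, and SplineThicken builds on that displaced position with particle-driven glow width.
# Particle color feeds emissive through an adjustment function.
# A fake fresnel from the U coordinate, peaking mid-strip and sharpened by an exponent, multiplies particle alpha into a depth fade.
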Material /BRCosmetics/Gadgets/Assets/VinderTech_GliderChute/Glider_ScorpionZero/FX/Materials/M_Glider_ScorpionZero_Wings_Glow
graph {
 properties {
  bUsedWithNiagaraMeshParticles=true
  blendMode=BLEND_Additive
  pixelDepthOffsetMode=PDOM_Legacy
  shadingModel=MSM_Unlit
  usedShadingModels=Unlit
 }
 nodes [
  n1 = DynamicParameter() {parameterIndex=1, paramNames [Glow Width, "", WPO Intensity, ""]}
  n2 = Multiply(
   A=@n1.WPO Intensity
   B=MaterialFunctionCall() {materialFunction=&/MF_Glider_ScorpionZero_WPO}.World Position Offset
  )
  n3 = WorldPosition()
  Add(
   A=MaterialFunctionCall(WidthBase=@n1.Glow Width, WidthTip=@n1.Glow Width, WorldPosition=Add(A=@n2, B=@n3.XYZ)) {materialFunction=&SplineThicken}.WorldPositionOffset
   B=@n2
  )
  n4 = ParticleColor()
  n5 = ConstantBiasScale(None=ComponentMask(None=TextureCoordinate()) {r=true}) {bias=-0.5, scale=2}
  n6 = OneMinus(None=DotProduct(A=@n5, B=@n5))
  n7 = ScalarParameter() {defaultValue=15, parameterName=Fresnel Exponent}
  DepthFade(Opacity=Multiply(A=@n4.A, B=Power(Base=@n6, Exp=@n7)))
  MaterialFunctionCall(Emissive=@n4.RGB) {materialFunction=&MF_FX_EmissiveAdjustments}
 ]
}
