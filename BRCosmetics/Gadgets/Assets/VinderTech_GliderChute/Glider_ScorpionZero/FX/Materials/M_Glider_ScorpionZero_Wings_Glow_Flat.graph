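# Additive, unlit, two-sided glow material for the Scorpion Zero glider's wing FX, usable on Niagara mesh particles.
# Particle color feeds a shared emissive adjustment function.
# A mask multiplies red vertex color by an inverse-fresnel term that fades at grazing angles, each with a tunable exponent.
# The fresnel uses the wing WPO function's normals and a view vector from the displaced position, scaled by a dynamic WPO Intensity parameter.
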Material /BRCosmetics/Gadgets/Assets/VinderTech_GliderChute/Glider_ScorpionZero/FX/Materials/M_Glider_ScorpionZero_Wings_Glow_Flat
graph {
 properties {
  bUsedWithNiagaraMeshParticles=true
  blendMode=BLEND_Additive
  pixelDepthOffsetMode=PDOM_Legacy
  shadingModel=MSM_Unlit
  twoSided=true
  usedShadingModels=Unlit
 }
 nodes [
  MaterialFunctionCall(Emissive=ParticleColor().RGB) {materialFunction=&MF_FX_EmissiveAdjustments}
  n1 = VertexColor()
  n2 = CameraPositionWS()
  n3 = MaterialFunctionCall() {materialFunction=&/MF_Glider_ScorpionZero_WPO}
  n4 = Add(
   A=Multiply(A=DynamicParameter() {parameterIndex=1, paramNames ["", "", WPO Intensity, ""]}.WPO Intensity, B=@n3.World Position Offset)
   B=WorldPosition().XYZ
  )
  n5 = DotProduct(A=Normalize(VectorInput=Subtract(A=@n2, B=@n4)), B=@n3.World Space Normals)
  Multiply(
   A=Power(Base=ComponentMask(None=@n1.R) {r=true}, Exp=ScalarParameter() {defaultValue=2, parameterName=Opacity Exponent})
   B=Power(Base=Abs(None=@n5), Exp=ScalarParameter() {defaultValue=2, parameterName=Fresnel Exponent})
  )
 ]
}
