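# Niagara system for the ScorpionZero glider's wing visuals, with glow, trails and deploy flares.
# Wings_Individual bursts two persistent mesh particles, split by partition into mirrored left and right wings across eight mesh renderers.
# System-level orientation, trail and curl values steer each wing, while fade-in and settle timers drive material parameters, visibility tags and flare curves.
# Glow brightness scales with time of day; a sprite debug emitter is disabled.
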NiagaraSystem /BRCosmetics/Gadgets/Assets/VinderTech_GliderChute/Glider_ScorpionZero/FX/NS_Glider_ScorpionZero
dependencies {
 usedDataInterfaces [
  &NiagaraDataInterfaceArrayFloat3
  &NiagaraDataInterfaceArrayFloat
  &NiagaraDataInterfaceArrayInt32
  &NiagaraDataInterfaceCurve
 ]
 usedDynamicInputs [
  &MakeVector2D
  &FN_FrontEndFloatModifier
  /Niagara/DynamicInputs/Math/Min_Float
  &Divide_Float
  &Max_Float
  &Subtract_Float
  &FNTimeOfDayScalar
  &Abs_Float
  &ClampFloat
  /Niagara/DynamicInputs/Math/LerpMultipleFloats
  &Add_Float
  &Power
 ]
 usedModules [
  &SystemState
  &self:SystemSpawnScript.SystemScriptSource.SystemScriptGraph.NiagaraNodeAssignment_0.SetVariables_#1
 ]
}
emitters {
 FN_Debug {
  bEnabled=false
  emitterUpdateScript [
   &EmitterState()
   &SpawnBurst_Instantaneous(Spawn Count=1, Use Spawn Probability=false)
  ]
  particleSpawnScript [/Niagara/Modules/Spawn/Initialization/InitializeParticle(
    Lifetime=5
    Module.Write Color=true
    Module.Write Lifetime=true
    Module.Write Mass=true
    Module.Write Material Random=true
    Module.Write Position=false
    Module.Write Scale=false
    Module.Write SpriteRotation=false
    Module.Write SpriteSize=true
    Position=&SimulationPosition()
    Sprite Size=<100 100>
   )]
  particleUpdateScript [
   &ParticleState()
   &FN_Timer()
   /SetVariables(Particles.Debug=0.1) {isSetParameters=true}
   &DynamicMaterialParameters(
    Index 0 Param 1=$Particles.Debug
    Module.Param0WriteEnabled=true
    Module.Param1WriteEnabled=true
    Module.Param2WriteEnabled=true
    Module.Param3WriteEnabled=true
   )
   &CameraOffset(Camera Offset Amount=111)
  ]
  rendererClasses [&NiagaraSpriteRendererProperties]
  renderers [{rendererClass=&NiagaraSpriteRendererProperties}]
 }
 Wings_Individual {
  emitterUpdateScript [
   &EmitterState()
   &SpawnBurst_Instantaneous(Spawn Count=2, Use Spawn Probability=false, UseLoopCountLimit=true)
  ]
  particleSpawnScript [
   &PartitionParticles()
   /Niagara/Modules/Spawn/Initialization/V2/InitializeParticle(
    Color Mode=::NewEnumerator0
    Material Random:NiagaraFloat=&Lerp_Float(
     A=&RandomRangeFloat(Minimum=0.25, Maximum=0.5)
     B=&RandomRangeFloat(Minimum=0.75, Maximum=1)
     Alpha=&MakeFloatFromInt(Int=$Particles.PartitionParticles.Partition)
    )
    Mesh Scale Mode=::NewEnumerator3
    Mesh Uniform Scale=0.85
    Position Offset=&SelectVectorFromArray(
     Direct Array Index=$Particles.PartitionParticles.Partition
     Direct Array Mode=::NewEnumerator1
     Vector Selection Array=DataInterface() {propertyValues {FloatData [<-40 3 7>, <40 3 7>]}}
    )
    UsePositionOffset=true
    Write Material Random=true
   )
   &InitialMeshOrientation(Mesh Orientation Mode=::NewEnumerator7, Rotation Coordinate Space=::NewEnumerator0, Use Rotation Vector=true)
  ]
  particleUpdateScript [
   &InterpolateOverTime(Rate Of Change=6, Target Float Value:NiagaraFloat=$User.RightDot)
   &ParticleState(Kill Particles When Lifetime Has Elapsed=false, Let Infinitely Lived Particles Die When Emitter Deactivates=true)
   &UpdateMeshOrientation(
    Facing Direction=&Add_Vector(
     A=&Lerp_Vector(
      A=<1 0 0>
      B=&MakeVector(X=$System.WingOrientationX, Y=$System.WingOrientationY, Z=0)
      Alpha=$System.FadeInSettleTimer
     )
     B=&SelectVectorFromArray(
      Direct Array Index=$Particles.PartitionParticles.Partition
      Vector Selection Array=DataInterface() {propertyValues {FloatData [<0 1 0>, <0 -1 0>]}}
     )
    )
    Facing Interpolation=false
    Orientation Method=::NewEnumerator1
    Override Quat=false
    Stabilization Mode:ENiagara_OrientVectorCount=::NewEnumerator0
    TransformFacing=false
   )
   &DynamicMaterialParameters(
    Index 0 Param 1=&SelectFloatFromArray(
     Direct Array Index=$Particles.PartitionParticles.Partition
     Direct Array Mode=::NewEnumerator1
     Float Selection Array=DataInterface() {propertyValues {FloatData [1,-1]}}
    )
    Index 0 Param 2=&SelectFloatFromArray(
     Direct Array Index=$Particles.PartitionParticles.Partition
     Direct Array Mode=::NewEnumerator1
     Float Selection Array=DataInterface() {propertyValues {FloatData [0,0.0666]}}
    )
    Index 0 Param 3=&Lerp_Float(
     A=$System.WingTrailLeft
     B=$System.WingTrailRight
     Alpha=&MakeFloatFromInt(Int=$Particles.PartitionParticles.Partition)
    )
    Index 0 Param 4=&Multiply_Float(
     A=$System.WingCurlIntensity
     B=&Multiply_FloatByInt(
      Float=&Lerp_Float(
       A=&Lerp_Float(A=$System.WingTrailRight, B=-0.5, Alpha=$System.WingCurlLerp)
       B=&Lerp_Float(A=$System.WingTrailLeft, B=-0.5, Alpha=$System.WingCurlLerp)
       Alpha=&MakeFloatFromInt(Int=$Particles.PartitionParticles.Partition)
      )
      Integer=&SelectIntFromArray(
       Direct Array Index=$Particles.PartitionParticles.Partition
       Direct Array Mode=::NewEnumerator1
       Int Selection Array=DataInterface() {propertyValues {IntData [-1,1]}}
      )
     )
    )
    Module.Index0Param0Write=true
    Module.Index0Param1Write=true
    Module.Index0Param2Write=true
    Module.Index0Param3Write=true
   )
   /SetVariables(Particles.WingTrailDynamicParameters1:Vector4f=&MakeVector4(
     X=&Multiply_Float(
      A=$System.FadeInSettleTimer
      B=&Lerp_Float(
       A=$System.WingTrailLeft
       B=$System.WingTrailRight
       Alpha=&MakeFloatFromInt(Int=$Particles.PartitionParticles.Partition)
      )
     )
     Y=0
     Z=0
     W=0
    )) {isSetParameters=true}
   /SetVariables(
    Particles.WingGlowColor:LinearColor=&MultiplyLinearColor(
     Linear Color=<rgba 0.3 0.216 0.09 0.1>
     Scale Factor=&MakeFloat4FromVectorAndFloat(
      "Vector (RGB)"=&Multiply_VectorByFloat(
       Vector=<1 1 1>
       Float=&FloatFromCurve(
        FloatCurve=DataInterface() {propertyValues {Curve {keys [<at 0 0 RCIM_Constant>,<at 0.99 1 RCIM_Constant>]}}}
        CurveIndex=$System.FadeInTimer
        Scale Curve=$System.TimeOfDayEmissiveMultiplier
       )
      )
      "Float (Alpha)"=$System.TimeOfDayEmissiveMultiplier
     )
    )
    Particles.WingGlowDynamicParameters1:Vector4f=&MakeVector4(
     X=&FloatFromCurve(
      FloatCurve=DataInterface() {propertyValues {Curve {keys [<at 0 0 RCIM_Constant>,<at 0.99 11 RCIM_Constant>]}}}
      CurveIndex=$System.FadeInTimer
     )
     Y=0
     Z=$System.FadeInSettleTimer
     W=0
    )
   ) {isSetParameters=true}
   /SetVariables(
    Particles.WingDynamicParameters1:Vector4f=&MakeVector4(
     X=&FloatFromCurve(
      FloatCurve=DataInterface() {propertyValues {Curve {keys [{interpMode=RCIM_Cubic, tangentMode=RCTM_User, time=0, value=0}, {interpMode=RCIM_Cubic, tangentMode=RCTM_User, time=1, value=1}]}}}
      CurveIndex=$System.FadeInTimer
     )
     Y=&FloatFromCurve(
      FloatCurve=DataInterface() {propertyValues {Curve {keys [
          {arriveTangent=9.0725, interpMode=RCIM_Cubic, leaveTangent=9.0725, tangentMode=RCTM_User, time=0, value=0}
          {arriveTangent=-228.788, interpMode=RCIM_Cubic, leaveTangent=-228.788, tangentMode=RCTM_User, time=0.05, value=99}
          {arriveTangent=-0.0405, interpMode=RCIM_Cubic, leaveTangent=-0.0405, tangentMode=RCTM_User, time=1, value=1}
         ]}}}
      CurveIndex=$System.FadeInSettleTimer
      Scale Curve=&Multiply_Float(A=1.5, B=$System.TimeOfDayEmissiveMultiplier)
     )
     Z=$System.FadeInSettleTimer
     W=0
    )
    Particles.WingVisibilityTag:NiagaraInt32=&MakeIntFromFloat(Float=$System.FadeInTimer, Conversion Mode=::NewEnumerator3)
   ) {isSetParameters=true}
   /SetVariables(
    Particles.FlareColor=&MakeLinearColorFromVectorAndFloat(
     "Vector (RGB)"=<1 0.75 0.5>
     "Float (Alpha)"=&FloatFromCurve(
      FloatCurve=DataInterface() {propertyValues {Curve {keys [
          {arriveTangent=-0.2998, interpMode=RCIM_Cubic, leaveTangent=-0.2998, tangentMode=RCTM_User, time=0, value=0}
          {arriveTangent=-2.5249, interpMode=RCIM_Cubic, leaveTangent=-2.5249, tangentMode=RCTM_User, time=0.1, value=1}
          {arriveTangent=-0.0405, interpMode=RCIM_Cubic, leaveTangent=-0.0405, tangentMode=RCTM_User, time=1, value=0}
         ]}}}
      CurveIndex=$System.FadeInSettleTimer
      Scale Curve=4
     )
    )
    Particles.FlarePopsDynamicParameters1:Vector4f=&MakeVector4(
     X=&FloatFromCurve(
      FloatCurve=DataInterface() {propertyValues {Curve {keys [
          {arriveTangent=4.1785, interpMode=RCIM_Cubic, time=0, value=0}
          {
           arriveTangent=1
           arriveTangentWeight=0.0667
           interpMode=RCIM_Cubic
           leaveTangent=1
           leaveTangentWeight=0.3605
           tangentMode=RCTM_User
           time=0.2
           value=0.7
          }
          {interpMode=RCIM_Cubic, leaveTangent=0.0224, time=1, value=1}
         ]}}}
      CurveIndex=$System.FadeInSettleTimer
      Scale Curve=175
     )
     Y=0.5
     Z=0
     W=0
    )
    Particles.FlareRainbowDynamicParameters1:Vector4f=&MakeVector4(
     X=&FloatFromCurve(
      FloatCurve=DataInterface() {propertyValues {Curve {keys [{tangentMode=RCTM_User, time=0, value=0}, <at 0.02 0.5>, {tangentMode=RCTM_User, time=1, value=1}]}}}
      CurveIndex=$System.FadeInSettleTimer
      Scale Curve=133
     )
     Y=&Lerp_Float(
      A=1
      B=0.266
      Alpha=&FloatFromCurve(
       FloatCurve=DataInterface() {propertyValues {Curve {keys [
           {arriveTangent=2.2345, interpMode=RCIM_Cubic, leaveTangent=2.2345, tangentMode=RCTM_User, time=0, value=0}
           {interpMode=RCIM_Cubic, tangentMode=RCTM_User, time=1, value=1}
          ]}}}
       CurveIndex=$System.FadeInSettleTimer
      )
     )
     Z=&FloatFromCurve(
      FloatCurve=DataInterface() {propertyValues {Curve {keys [{arriveTangent=-3.323, interpMode=RCIM_Cubic, leaveTangent=-3.323, tangentMode=RCTM_User, time=0, value=1}, <at 1 0 RCIM_Cubic>]}}}
      CurveIndex=$System.FadeInSettleTimer
     )
     W=0
    )
    Particles.FlareScale:Vector3f=&VectorFromFloat(Value=&FloatFromCurve(FloatCurve=DataInterface() {propertyValues {Curve {keys [<at 0 0>,<at 0.01 1 RCIM_Constant>]}}}, CurveIndex=$System.FadeInSettleTimer))
    Particles.FlareVisibilityTag:NiagaraInt32=&MakeIntFromFloat(Float=$System.FadeInSettleTimer, Conversion Mode=::NewEnumerator3)
   ) {isSetParameters=true}
  ]
  rendererClasses [&NiagaraMeshRendererProperties]
  renderers [
   {rendererClass=&NiagaraMeshRendererProperties, rendererIndex=0}
   {rendererClass=&NiagaraMeshRendererProperties, rendererIndex=1}
   {rendererClass=&NiagaraMeshRendererProperties, rendererIndex=2}
   {rendererClass=&NiagaraMeshRendererProperties, rendererIndex=3}
   {rendererClass=&NiagaraMeshRendererProperties, rendererIndex=4}
   {rendererClass=&NiagaraMeshRendererProperties, rendererIndex=5}
   {rendererClass=&NiagaraMeshRendererProperties, rendererIndex=6}
   {rendererClass=&NiagaraMeshRendererProperties, rendererIndex=7}
  ]
 }
}
summary {
 systemName=NS_Glider_ScorpionZero
 emitters [
  {emitterName=Wings_Individual, rendererClasses [&NiagaraMeshRendererProperties]}
  {bEnabled=false, emitterName=FN_Debug, rendererClasses [&NiagaraSpriteRendererProperties]}
 ]
 userVariables [
  User.bIsFrontEnd {type=NiagaraBool}
  User.bIsFrontEndPreview {type=NiagaraBool}
  User.bIsFullyDeployed {type=NiagaraBool, defaultValue=true}
  User.FadeInDuration {type=NiagaraFloat, defaultValue=0.666, description=""}
  User.FadeInSettleDuration {type=NiagaraFloat, defaultValue=1.1, description=""}
  User.ForwardDot {type=NiagaraFloat, defaultValue=0, description=""}
  User.PlayerSpeed {type=NiagaraFloat, defaultValue=0, description=""}
  User.RibbonIDOffset {type=NiagaraInt32}
  User.RightDot {type=NiagaraFloat, defaultValue=0, description=""}
  User.SkeletalMesh {
   type=NiagaraDataInterfaceSkeletalMesh
   defaultValue {
    struct=NiagaraExt_VariableValue_DataInterface
    value {dataInterface=&self:NiagaraDataInterfaceSkeletalMesh_0}
   }
  }
  User.StartDelay {type=NiagaraFloat, defaultValue=1.4, description=""}
 ]
}
